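# Niagara swing trail for the KeyTracker pickaxe, built from three layered ribbon emitters: Trail_Cyan, Trail_Orange and Trail_Screen.
# Each uses AnimTrail modules to spawn, initialize and update ribbons, with a skeletal mesh exposed as a user parameter.
# Lifetime curves taper width, shift color and fade alpha.
# Time of day adjusts Cyan's base color and Screen's alpha, while Orange is narrower, shorter-lived and camera-offset.
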NiagaraSystem /BRCosmetics/Weapons/FORT_Melee/Pickaxe_KeyTracker/FX/NS_Pickaxe_KeyTracker_Trail
dependencies {
 usedDataInterfaces [&NiagaraDataInterfaceCurve, &NiagaraDataInterfaceColorCurve]
 usedDynamicInputs [&Multiply_Float]
 usedModules [&SystemState]
}
emitters {
 Trail_Cyan {
  emitterUpdateScript [
   &EmitterState(Life Cycle Mode=::NewEnumerator1, Loop Behavior=::NewEnumerator1, Loop Duration=0.27, UseLoopDelay=true)
   &FN_SpawnAnimTrail_v2()
  ]
  particleSpawnScript [
   /Game/Effects/Niagara/Modules/AnimTrail/FN_InitializeAnimTrail/FN_InitializeAnimTrailBoneIndecies(First Bone or Socket Index=1)
   /Game/Effects/Niagara/Modules/AnimTrail/FN_InitializeAnimTrail/FN_InitializeAnimTrail(
    AnimTrailWidth=1
    Color=&FNTimeOfDayColor(Night Color=<rgba 1 1 1 0.1>, Morning Color=<rgba 1.5 1.5 1.5 0.1>, Day Color=<rgba 1.5 1.5 1.5 0.1>, Evening Color=<rgba 1 1 1 0.1>)
    Lifetime=0.22
   )
   &SystemLocation(MaskSpawnGroup=false)
  ]
  particleUpdateScript [
   &ScaleRibbonWidth(Ribbon Width Scale=&FloatFromCurve(FloatCurve=DataInterface() {propertyValues {Curve {keys [
         {arriveTangent=-1.2579, interpMode=RCIM_Cubic, time=0, value=1}
         {arriveTangent=-0.9664, interpMode=RCIM_Cubic, leaveTangent=-0.9664, tangentMode=RCTM_User, time=1, value=0}
        ]}}}))
   &ScaleColor(
    Scale Mode=::NewEnumerator1
    Scale RGBA=&Vector4FromCurve(Vector4Curve=DataInterface() {propertyValues {
       AlphaCurve {keys [<at 0 1>,<at 1 0>]}
       BlueCurve {keys [<at 0.2072 1>,<at 1 1>]}
       GreenCurve {keys [<at 0.2072 0.5836>,<at 1 1>]}
       RedCurve {keys [<at 0.2072 0.259>,<at 1 1>]}
      }})
    ScaleRGBA=true
   ) {module=ScaleColor002}
   &FN_UpdateAnimTrail_v2()
  ]
  rendererClasses [&NiagaraRibbonRendererProperties]
  renderers [{rendererClass=&NiagaraRibbonRendererProperties}]
 }
 Trail_Orange {
  emitterUpdateScript [
   &EmitterState(Life Cycle Mode=::NewEnumerator1, Loop Behavior=::NewEnumerator1, Loop Duration=0.27, UseLoopDelay=true)
   &FN_SpawnAnimTrail_v2(AnimTrailWidthMode=::FromSecond)
  ]
  particleSpawnScript [
   &FN_InitializeAnimTrailBoneIndecies(First Bone or Socket Index=1)
   &FN_InitializeAnimTrail(AnimTrailWidth=0.75, Color=<rgba 0.5 0.5 0.5 0.2>, Lifetime=0.15)
   &SystemLocation(MaskSpawnGroup=false)
   &CameraOffset(Camera Offset Amount=2)
  ]
  particleUpdateScript [
   &ScaleRibbonWidth(Ribbon Width Scale=&FloatFromCurve(FloatCurve=DataInterface() {propertyValues {Curve {keys [
         {arriveTangent=-1.2579, interpMode=RCIM_Cubic, time=0, value=1}
         {arriveTangent=-0.9664, interpMode=RCIM_Cubic, leaveTangent=-0.9664, tangentMode=RCTM_User, time=1, value=0}
        ]}}}))
   &ScaleColor(
    Scale Mode=::NewEnumerator1
    Scale RGBA=&Vector4FromCurve(Vector4Curve=DataInterface() {propertyValues {
       AlphaCurve {keys [<at 0 1>,<at 1 0>]}
       BlueCurve {keys [<at 0.0639 2.3238>,<at 0.3893 1.074>,<at 1 1>]}
       GreenCurve {keys [<at 0.0639 1.236>,<at 0.3893 1.5418>,<at 1 1>]}
       RedCurve {keys [<at 0.0639 3>,<at 0.3893 3>,<at 1 1>]}
      }})
    ScaleRGBA=true
   ) {module=ScaleColor002}
   &FN_UpdateAnimTrail_v2()
  ]
  rendererClasses [&NiagaraRibbonRendererProperties]
  renderers [{rendererClass=&NiagaraRibbonRendererProperties}]
 }
 Trail_Screen {
  emitterUpdateScript [
   &EmitterState(Life Cycle Mode=::NewEnumerator1, Loop Behavior=::NewEnumerator1, Loop Duration=0.5, UseLoopDelay=false)
   &FN_SpawnAnimTrail_v2(AnimTrailWidthMode=::FromSecond)
  ]
  particleSpawnScript [
   &FN_InitializeAnimTrailBoneIndecies(First Bone or Socket Index=1)
   &FN_InitializeAnimTrail(AnimTrailWidth=0.25, Color=<rgba 0.5 0.5 0.5 1>)
  ]
  particleUpdateScript [
   &ParticleState()
   &FN_UpdateAnimTrail_v2()
   &ScaleColor(
    Scale Alpha=&FloatFromCurve(
     FloatCurve=DataInterface() {propertyValues {Curve {keys [{interpMode=RCIM_Cubic, tangentMode=RCTM_User, time=0, value=1}, {interpMode=RCIM_Cubic, tangentMode=RCTM_User, time=1, value=0}]}}}
     Scale Curve=&FNTimeOfDayScalar(Night Scale=0.8, Morning Scale=0.3, Evening Scale=0.4)
    )
    Scale RGB=&MakeVectorFromLinearColorRGB(Color=&ColorFromCurve(ColorCurve=DataInterface() {propertyValues {
        AlphaCurve {keys []}
        BlueCurve {keys [
          <at 0.097 1>
          {arriveTangent=-0.7112, interpMode=RCIM_Cubic, leaveTangent=-0.7112, time=0.3445, value=1}
          <at 0.6524 0.605>
          <at 0.9229 0.7015>
         ]}
        GreenCurve {keys [
          <at 0.097 0.432>
          {arriveTangent=0.5153, interpMode=RCIM_Cubic, leaveTangent=0.5153, time=0.3445, value=0.6146}
          <at 0.6524 0.7182>
          <at 0.9229 1>
         ]}
        RedCurve {keys [
          <at 0.097 0.716>
          {arriveTangent=0.5113, interpMode=RCIM_Cubic, leaveTangent=0.5113, time=0.3445, value=0.511}
          <at 0.6524 1>
          <at 0.9229 0.7053>
         ]}
       }}))
    ScaleA=true
    ScaleRGB=true
   )
   &ScaleRibbonWidth(Ribbon Width Scale=&FloatFromCurve(FloatCurve=DataInterface() {propertyValues {Curve {keys [{interpMode=RCIM_Cubic, tangentMode=RCTM_User, time=0, value=1}, {interpMode=RCIM_Cubic, tangentMode=RCTM_User, time=1, value=0}]}}}))
  ]
  rendererClasses [&NiagaraRibbonRendererProperties]
  renderers [{rendererClass=&NiagaraRibbonRendererProperties}]
 }
}
summary {
 systemName=NS_Pickaxe_KeyTracker_Trail
 emitters [
  {emitterName=Trail_Cyan, rendererClasses [&NiagaraRibbonRendererProperties]}
  {emitterName=Trail_Orange, rendererClasses [&NiagaraRibbonRendererProperties]}
  {emitterName=Trail_Screen, rendererClasses [&NiagaraRibbonRendererProperties]}
 ]
 userVariables [
  User.RibbonIDOffset {type=NiagaraInt32}
  User.SkeletalMesh {
   type=NiagaraDataInterfaceSkeletalMesh
   defaultValue {
    struct=NiagaraExt_VariableValue_DataInterface
    value {dataInterface=&self:NiagaraDataInterfaceSkeletalMesh_0}
   }
  }
 ]
}
